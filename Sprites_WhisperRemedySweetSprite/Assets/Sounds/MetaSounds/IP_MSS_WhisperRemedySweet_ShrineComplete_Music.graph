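# Stereo one-shot MetaSound source that plays the Super Sprite Shrine music sting, named for completion of the Whisper Remedy Sweet shrine.
# On play, an ADSR envelope starts the Wave Player partway into the sting and fades it in to full sustain.
# The envelope times a decibel-variable gain scales both channels, and the source finishes with the wave.
# It borrows Fire Petal Shrine attenuation and overrides concurrency to limit simultaneous instances.
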MetaSoundSource /Sprites_WhisperRemedySweetSprite/Assets/Sounds/MetaSounds/IP_MSS_WhisperRemedySweet_ShrineComplete_Music
inputs [UE.Source.OnPlay {type=Trigger}]
interfaces [UE.OutputFormat.Stereo:1.0, UE.Source.OneShot:1.0, UE.Source:1.1]
outputs [
 "UE.OutputFormat.Stereo.Audio:0" {type=Audio}
 "UE.OutputFormat.Stereo.Audio:1" {type=Audio}
 UE.Source.OneShot.OnFinished {type=Trigger}
]
pages [Default {
  input_defaults {UE.Source.OnPlay=null}
  nodes [
   ADSREnvelopeADSREnvelope_0 = ADSR Envelope.ADSR Envelope:Float(Trigger Attack=^UE.Source.OnPlay) {also {defaults {Attack Time=0.5, Sustain Level=1}}}
   DecibelstoLinearGain_0 = UE.Decibels to Linear Gain:Float(Decibels=VariableAccessor.Float(Variable=InitVariable.Float().Variable).Value)
   n1 = UE.Multiply:Float(PrimaryOperand=@ADSREnvelopeADSREnvelope_0.Out Envelope, AdditionalOperands=@DecibelstoLinearGain_0.Linear Gain)
   WavePlayer_0 = UE.Wave Player:Stereo(Play=@ADSREnvelopeADSREnvelope_0.On Attack Triggered) {also {defaults {Start Time=2, Wave Asset=/Sprites_WhisperRemedySweetSprite/Assets/Sounds/Waves/Music/MX_S36_POI_SuperSpriteShrine_STNG}}}
   Multiply_0 = UE.Multiply:Audio by Float(AdditionalOperands=@n1.Out, PrimaryOperand=@WavePlayer_0.Out Right)
   Multiply_1 = UE.Multiply:Audio by Float(AdditionalOperands=@n1.Out, PrimaryOperand=@WavePlayer_0.Out Left)
  ]
  links [
   {dst="@UE.OutputFormat.Stereo.Audio:1", src=Multiply_0.Out}
   {dst="@UE.OutputFormat.Stereo.Audio:0", src=Multiply_1.Out}
   {dst="@UE.Source.OneShot.OnFinished", src=WavePlayer_0.On Finished}
  ]
 }]
properties {
 assetImportData=&self:AssetImportData
 attenuationSettings=/FirePetalShrineGameplay/Assets/Sounds/Settings/AS_FirePetalShrine_Activate_Music
 bOverrideConcurrency=true
 maxDistance=6100
 numChannels=2
 outputFormat=Stereo
 sampleRate=48000
 soundClassObject=&Fort_Loot_Reward_PSM
 concurrencyOverrides {maxCount=2}
}
